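# Parent material for the Manno SE vehicle wheel, rendered as unlit additive on skeletal meshes with a scalar opacity.
# Vertex red blends the rim normal with a remapped copy to drive a classic Fresnel term borrowed from the Cola Holographic wheel.
# That tinted term adds to a rim-texture emissive mask with brightness control; a reflection-vector term is multiplied by zero.
# The result passes through the shared vehicle damage and de-rez transition functions.
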Material /VehicleCosmetics/Wheels/Manno_SE/Materials/M_WheelParent_Manno_SE
graph {
 properties {
  bAutomaticallySetUsageInEditor=false
  bUseMaterialAttributes=true
  bUsedWithSkeletalMesh=true
  blendMode=BLEND_Additive
  shadingModel=MSM_Unlit
  usedShadingModels=Unlit
 }
 nodes [
  n1 = TextureSampleParameter2D() {parameterName=Rim_Normal, samplerType=SAMPLERTYPE_Normal, texture=/VehicleCosmetics/Wheels/Manno_SE/Textures/T_MannoSe_Rim_N}
  n2 = OneMinus(None=VertexColor().R)
  n3 = Add(
   A=Multiply(A=@n1.RGBA, B=Constant4Vector() {constant=<rgba 0.625 0.625 1 0>}.RGBA)
   B=Constant4Vector() {constant=<rgba 0.375 0.375 0 0>}.RGBA
  )
  n4 = Multiply(
   A=NamedRerouteUsage()
   B=MaterialFunctionCall!(Normal=LinearInterpolate(A=@n1.RGBA, B=@n3, Alpha=@n2), Exponent=ScalarParameter() {defaultValue=1.5, parameterName=Fresnel_Exponent}) {materialFunction=/VehicleCosmetics/Wheels/Cola_Holographic/Materials/MF_Fresnel_Classic}.Result
  )
  n5 = NamedRerouteUsage()
  n6 = Multiply(A=Desaturation(None=NamedRerouteUsage(), Fraction=Constant() {r=0.2}), B=LinearInterpolate(A=@n5, B=@n5))
  n7 = ReflectionVectorWS()
  n8 = Constant()
  n9 = TextureSampleParameter2D() {parameterName=Rim_RGB, samplerType=SAMPLERTYPE_LinearColor, texture=/VehicleCosmetics/Wheels/Manno_SE/Textures/T_MannoSe_Rim_SE_RGB}
  n10 = TextureSampleParameter2D() {parameterName=Rim_E, samplerType=SAMPLERTYPE_LinearColor, texture=/VehicleCosmetics/Wheels/Manno_SE/Textures/T_MannoSe_Rim_SE_E}
  n11 = LinearInterpolate(A=@n9.A, B=Multiply(A=@n10.R) {constB=0.75})
  n12 = NamedRerouteUsage()
  n13 = Add(
   A=Add(A=Multiply(A=@n4, B=@n6), B=Multiply(A=@n7, B=@n8))
   B=Multiply(A=Multiply(A=@n11, B=@n12), B=ScalarParameter() {defaultValue=3, parameterName=EmissiveBrightness})
  )
  n14 = ScalarParameter() {defaultValue=0.33, parameterName=Opacity}
  MaterialFunctionCall(In=MaterialFunctionCall(Material=SetMaterialAttributes!(MaterialAttributes, Emissive Color=@n13, Opacity=@n14) {attributeSetTypes [#1, #2]}) {materialFunction=&MF_BaseVehicle_Damage}.Result) {materialFunction=&MF_VehicleDeRez_Transition}
  NamedRerouteDeclaration(None=VectorParameter() {parameterName=EmmisiveColor, defaultValue=<rgba 1 0.0377 0.0377 0>}.RGB) {name=EmCol, nodeColor=<rgba 1 0 0.6328 1>}
  NamedRerouteDeclaration(None=@n1.B) {name=Norm_B, nodeColor=<rgba 0 0.9531 1 1>}
 ]
}
