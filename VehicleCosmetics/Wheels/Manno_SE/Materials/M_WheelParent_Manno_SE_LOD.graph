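# Parent material for the Manno SE vehicle wheel LOD, adding masked surface colors and an emissive rim glow.
# MF_WheelParent_Simple supplies base attributes, which are overridden using mask textures, surface parameters and named reroutes.
# Emissive combines a Fresnel term borrowed from the Cola_Holographic wheel with rim textures and a brightness parameter, blended by vertex color red.
# The result passes through the shared vehicle damage function, then the DeRez transition.
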Material /VehicleCosmetics/Wheels/Manno_SE/Materials/M_WheelParent_Manno_SE_LOD
graph {
 properties {
  bAutomaticallySetUsageInEditor=false
  bUseMaterialAttributes=true
  bUsedWithMeshDeformer=true
  bUsedWithSkeletalMesh=true
  pixelDepthOffsetMode=PDOM_Legacy
  usedShadingModels=Default Lit
 }
 nodes [
  n1 = TextureSampleParameter2D() {parameterName=Texture_Mask, samplerType=SAMPLERTYPE_LinearColor, texture=/VehicleCosmetics/Wheels/Manno_SE/Textures/T_MannoSe_Rim_M_LOD}
  n2 = OneMinus(None=@n1.A)
  n3 = Add(A=@n2, B=@n1.R)
  Add(A=@n3, B=@n1.G)
  n4 = MaterialFunctionCall() {materialFunction=&MF_WheelParent_Simple}
  n5 = GetMaterialAttributes(None=@n4.MaterialAttributes) {attributeGetTypes [
    #1
    #2
    #3
    #4
    #5
   ]}
  n6 = VectorParameter() {group=4_Surface Defaults, parameterName=Color_Mask, defaultValue=<rgba 0.3108 0.2862 0.2832 1>}
  n7 = NamedRerouteUsage()
  n8 = SetMaterialAttributes!(
   MaterialAttributes=@n5.MaterialAttributes
   Emissive Color=LinearInterpolate(A=@n5.Emissive Color, B=NamedRerouteUsage(), Alpha=OneMinus(None=@n4.VertexColorR))
   Base Color=LinearInterpolate(A=LinearInterpolate(A=@n5.Base Color, B=@n6.RGB, Alpha=@n7), B=NamedRerouteUsage(), Alpha=NamedRerouteUsage())
   Metallic=NamedRerouteUsage()
   Roughness=NamedRerouteUsage()
   Normal=NamedRerouteUsage()
  ) {attributeSetTypes [
    #1
    #2
    #3
    #4
    #5
   ]}
  MaterialFunctionCall(In=MaterialFunctionCall(Material=@n8) {materialFunction=&MF_BaseVehicle_Damage}.Result) {materialFunction=&MF_VehicleDeRez_Transition}
  n9 = NamedRerouteUsage()
  n10 = TextureSampleParameter2D() {
   group=1_Input Textures
   parameterName=Rim_Normal
   samplerType=SAMPLERTYPE_Normal
   texture=/VehicleCosmetics/Wheels/Manno_SE/Textures/T_MannoSe_Rim_N_LOD
  }
  n11 = Multiply(A=@n10.RGBA, B=Constant4Vector() {constant=<rgba 0.625 0.625 1 0>}.RGBA)
  n12 = Constant4Vector() {constant=<rgba 0.375 0.375 0 0>}
  n13 = VertexColor()
  n14 = MaterialFunctionCall!(
   Normal=LinearInterpolate(A=NamedRerouteUsage(), B=Add(A=@n11, B=@n12.RGBA), Alpha=OneMinus(None=@n13.R))
   Exponent=ScalarParameter() {defaultValue=1.5, parameterName=Fresnel_Exponent}
  ) {materialFunction=/VehicleCosmetics/Wheels/Cola_Holographic/Materials/MF_Fresnel_Classic}
  n15 = Desaturation(None=NamedRerouteUsage(), Fraction=Constant() {r=0.2})
  n16 = NamedRerouteUsage()
  n17 = Add(A=Multiply(A=Multiply(A=@n9, B=@n14.Result), B=Multiply(A=@n15, B=@n16)), B=Multiply(A=ReflectionVectorWS()) {constB=0})
  n18 = TextureSampleParameter2D() {
   group=1_Input Textures
   parameterName=Rim_RGB
   samplerType=SAMPLERTYPE_LinearColor
   texture=/VehicleCosmetics/Wheels/Manno_SE/Textures/T_MannoSe_Rim_SE_RGB_LOD
  }
  n19 = TextureSampleParameter2D() {
   group=1_Input Textures
   parameterName=Rim_E
   samplerType=SAMPLERTYPE_LinearColor
   texture=/VehicleCosmetics/Wheels/Manno_SE/Textures/T_MannoSe_Rim_SE_E
  }
  n20 = Multiply(A=@n19.R) {constB=0.75}
  n21 = Multiply(
   A=Multiply(A=LinearInterpolate(A=@n18.A, B=@n20), B=NamedRerouteUsage())
   B=ScalarParameter() {defaultValue=3, parameterName=EmissiveBrightness}
  )
  NamedRerouteDeclaration(None=LinearInterpolate(A=Add(A=@n17, B=@n21), B=ScalarParameter() {parameterName=EmissiveMask}, Alpha=NamedRerouteUsage())) {name=Emissive, nodeColor=<rgba 1 0.75 0 1>}
  NamedRerouteDeclaration(None=VectorParameter() {parameterName=EmmisiveColor, defaultValue=<rgba 1 0.0377 0.0377 0>}.RGB) {name=EmCol, nodeColor=<rgba 0 0.5547 1 1>}
  n22 = ScalarParameter() {group=4_Surface Defaults, parameterName=Black_mask}
  NamedRerouteDeclaration(None=@n22) {name=BlackMask, nodeColor=<rgba 0 0.1562 1 1>}
  NamedRerouteDeclaration(None=@n10.RGBA) {name=Norm, nodeColor=<rgba 0.4766 1 0 1>}
  NamedRerouteDeclaration(None=@n10.B) {name=Norm_B, nodeColor=<rgba 1 0 0.7969 1>}
  NamedRerouteDeclaration(None=@n2) {name=Mask-1, nodeColor=<rgba 0 1 0.8828 1>}
  n23 = ScalarParameter() {defaultValue=1, group=4_Surface Defaults, parameterName=Metallic_Mask}
  n24 = NamedRerouteUsage()
  n25 = NamedRerouteUsage()
  NamedRerouteDeclaration(None=LinearInterpolate(A=LinearInterpolate(B=@n23, Alpha=@n24), B=NamedRerouteUsage(), Alpha=@n25)) {name=metalic, nodeColor=<rgba 1 0.5625 0 1>}
  n26 = ScalarParameter() {group=4_Surface Defaults, parameterName=Roughness_Mask}
  NamedRerouteDeclaration(None=LinearInterpolate(A=LinearInterpolate(B=@n26, Alpha=@n24) {constA=1}, B=OneMinus(None=@n22), Alpha=@n25)) {name=Roughness, nodeColor=<rgba 0.2422 0 1 1>}
  NamedRerouteDeclaration(None=@n3) {name=mask, nodeColor=<rgba 0.0781 1 0 1>}
 ]
}
